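; Golf cart (ATK) vehicle Blueprint for the Rufus map, adding water effects, drift/boost trail visuals and impulse ejection.
; Entering water spawns cosmetic splash, wake and audio, with a 0.25s timer updating height, direction and speed-based volume; overlapped lake type sets the height offset.
; Powerslide, boost and wheel-ground events drive trail particle parameters.
; On authority, strong impulses or impulse-ability impacts eject all players, gated by scalable-float settings.
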(event Vehicle|Water|EventOnEnteredWaterVolume
  (params (WaterSurfacePoint :type Vector%20%28by%20ref%29 :default "0, 0, 0"))
  (bind _returnvalue (CallFunction|GetWaterLevel))
  (bind (_returnvalue_x _returnvalue_y _returnvalue_z) (Math|Vector|BreakVector (Transformation|GetActorLocation)))
  (bind _watervfxoffset (Variables|Default|GetWaterVFXOffset))
  (bind _returnvalue_2 (+ _returnvalue _watervfxoffset))
  (bind _returnvalue_3 (- _returnvalue_2 _returnvalue_z))
  (bind _skeletalmesh (Variables|Vehicle|GetSkeletalMesh))
  ;[1 1 1] FX for water ineraction
  (Networking|CanExecuteCosmeticEvents
    (:True
      (if (> (Rendering|GetRenderingDetailMode) 0)
        (bind _returnvalue_4 (Utilities|Time|SetTimerbyEvent (delegate Custom|Check%20Water%20Loop) 0.25 true))
        (bind _inwaterfx (Variables|Default|GetInWaterFX)
          (:then
            (Components|DestroyComponent _inwaterfx))
          (:else
            _))
        (bind _returnvalue_5 (Effects|SpawnEmitterAttached :EmitterTemplate "/Game/WaterAndWind/Particles/WaterInteraction/P_Vehicle_In_Water_Athena.P_Vehicle_In_Water_Athena" :AttachToComponent _skeletalmesh :AttachPointName "None" :Location (Math|Vector|MakeVector 0.0 0.0 _returnvalue_3) :Scale "1.000000,1.000000,1.000000" :LocationType "KeepRelativeOffset" :bAutoDestroy true :PoolingMethod "None" :bAutoActivate true)) ; Looping water FX
        (Variables|Default|SetInWaterFX _returnvalue_5)
        (Effects|SpawnEmitterAttached :EmitterTemplate "/Game/WaterAndWind/Particles/WaterInteraction/P_WaterSplash_Athena_Vehicle.P_WaterSplash_Athena_Vehicle" :AttachToComponent _skeletalmesh :AttachPointName "None" :Location (Math|Vector|MakeVector 0.0 0.0 _returnvalue_3) :Scale "1.000000,1.000000,1.000000" :LocationType "KeepRelativeOffset" :bAutoDestroy true :PoolingMethod "None" :bAutoActivate true) ; Splash burst FX
        (bind _inwaterloop (Variables|Default|GetInWaterLoop)
          (:then
            (Audio|Components|Audio|Stop _inwaterloop))
          (:else
            _))
        (Audio|SpawnSoundAttached "/Game/Athena/Sounds/Vehicles/GolfCart/GolfCart_Movement_Impact_Water_Cue.GolfCart_Movement_Impact_Water_Cue" (Variables|Vehicle|GetSkeletalMesh))
        (bind _returnvalue_6 (Audio|SpawnSoundAttached "/Game/Athena/Sounds/Vehicles/GolfCart/GolfCart_Movement_InWater_Loop_Cue.GolfCart_Movement_InWater_Loop_Cue" (Variables|Vehicle|GetSkeletalMesh)))
        (Variables|Default|SetInWaterLoop _returnvalue_6)))))

(event Custom|Check%20Water%20Loop
  (bind (_returnvalue_x _returnvalue_y _returnvalue_z) (Math|Vector|BreakVector (Transformation|GetActorLocation)))
  ;[1 1 1] FX for water ineraction
  (bind _inwaterfx (Variables|Default|GetInWaterFX)
    (:then
      (Transformation|SetRelativeLocation _inwaterfx (Math|Vector|MakeVector 0.0 0.0 (- (+ (CallFunction|GetWaterLevel) (Variables|Default|GetWaterVFXOffset)) _returnvalue_z)) false false)
      (Effects|Components|ParticleSystem|SetVectorParameter (Variables|Default|GetInWaterFX) "Offset" (Math|Vector|Normalize (Math|Transform|InverseTransformDirection (Transformation|GetWorldTransform (Variables|Default|GetInWaterFX)) (Transformation|GetVelocity))))
      (bind _inwaterloop (Variables|Default|GetInWaterLoop)
        (:then
          (Audio|Components|Audio|SetVolumeMultiplier _inwaterloop (Math|Float|MapRangeClamped (Vehicle|GetForwardSpeedKmh) 0.0 2e+01 0.0 1.0) :!class /Script/Engine.AudioComponent))
        (:else)))
    (:else)))

(event Vehicle|Water|EventOnExitedWaterVolume
  (params (WaterSurfacePoint :type Vector%20%28by%20ref%29 :default "0, 0, 0"))
  ;[1 1 1] FX for water ineraction
  (Networking|CanExecuteCosmeticEvents
    (:True
      (Utilities|Time|ClearandInvalidateTimerbyHandle (stale Vehicle|Water|EventOnEnteredWaterVolume _returnvalue_4))
      (bind _inwaterfx (Variables|Default|GetInWaterFX)
        (:then
          (Components|Activation|SetActive _inwaterfx false false)
          (bind _inwaterloop (Variables|Default|GetInWaterLoop)
            (:then
              (Audio|Components|Audio|FadeOut _inwaterloop 3.0 0.0 "Linear" :!class /Script/Engine.AudioComponent))
            (:else)))
        (:else)))))

(event Collision|EventActorBeginOverlap
  (params (OtherActor :type Actor%20Object%20Reference))
  (bind _asbp_rufus_water (Utilities|Casting|CastToBP_Rufus_Water OtherActor))
  (Default|UpdateWaterVFXOffset true _asbp_rufus_water))

(event EventPowerSlideBegin
  (bind _ns_rufsusatktrails (Variables|Default|GetNS_RufsusATKTrails))
  (|CallParentFunction)
  (Effects|Components|ParticleSystem|SetVectorParameter _ns_rufsusatktrails "BoostColor" "0, 0, 0")
  (Effects|Components|ParticleSystem|SetBoolParameter _ns_rufsusatktrails "Sliding" true))

(event EventBoostBegin
  (|CallParentFunction)
  (Effects|Components|ParticleSystem|SetVectorParameter (Variables|Default|GetNS_RufsusATKTrails) "BoostColor" (Math|Conversions|ToVector(LinearColor) (Utilities|Array|Get(acopy) (Variables|WheelFX|GetBoostSlideColors)))))

(event Collision|EventActorEndOverlap
  (params (OtherActor :type Actor%20Object%20Reference))
  (bind _asbp_rufus_water (Utilities|Casting|CastToBP_Rufus_Water OtherActor))
  (Default|UpdateWaterVFXOffset false _asbp_rufus_water))

(event EventBoostEnd
  (|CallParentFunction)
  (Effects|Components|ParticleSystem|SetVectorParameter (Variables|Default|GetNS_RufsusATKTrails) "BoostColor" (Math|Conversions|ToVector(LinearColor) (Utilities|Array|Get(acopy) (Variables|WheelFX|GetBoostSlideColors)))))

(event EventPowerSlideEnd
  (|CallParentFunction)
  (Effects|Components|ParticleSystem|SetBoolParameter (Variables|Default|GetNS_RufsusATKTrails) "Sliding" false))

(event EventBoostIncrease
  (params (AccumulationStepIndex :type Integer))
  (|CallParentFunction AccumulationStepIndex)
  (Effects|Components|ParticleSystem|SetVectorParameter (Variables|Default|GetNS_RufsusATKTrails) "BoostColor" (Math|Conversions|ToVector(LinearColor) (Utilities|Array|Get(acopy) (Variables|WheelFX|GetBoostSlideColors) AccumulationStepIndex))))

(event EventTick
  (params (DeltaSeconds :type Float%20%28single-precision%29))
  (|CallParentFunction DeltaSeconds)
  (Effects|Components|ParticleSystem|SetBoolParameter (Variables|Default|GetNS_RufsusATKTrails) "Grounded" (Vehicle|WheelsonGround)))

(event EventOnHandleImpulsed
  (params (ImpulsePoint :type Vector%20%28by%20ref%29 :default "0, 0, 0") (ImpulseValue :type Vector%20%28by%20ref%29 :default "0, 0, 0") (InTags :type Gameplay%20Tag%20Container%20Structure%20%28by%20ref%29) (EffectContext :type Gameplay%20Effect%20Context%20Handle%20Structure%20%28by%20ref%29) (EventInstigator :type Controller%20Object%20Reference) (ImpulseCauser :type Actor%20Object%20Reference))
  ;[0.2812 0.2242 0.1087] When hit with a large normal explosion eject all players.
  (Networking|SwitchHasAuthority
    (:Authority
      (if (and (>= (Math|Vector|VectorLengthSquared ImpulseValue) (Math|Float|Square (ScalableFloat|GetValueatLevel (Variables|Default|GetHF_EjectOnImpulseThreshold)))) (ScalableFloat|GetValueasBool (Variables|Default|GetHF_EjectOnImpulse)))
        (Vehicle|EjectAllPlayers)))))

(event EventOnAbilityImpactAtLocation
  (params (WorldContextObject :type Object%20Reference) (InSound :type Sound%20Base%20Object%20Reference) (InInstigator :type Object%20Reference) (Velocity :type Vector%20%28by%20ref%29 :default "0, 0, 0") (Radius :type Float%20%28single-precision%29) (Tag :type Name :default "None") (bSubtractLocalNoise :type Boolean))
  ;[0.2812 0.2242 0.1087] When impacted by an impulsing ability, eject all players.
  (Networking|SwitchHasAuthority
    (:Authority
      (if (ScalableFloat|GetValueasBool (Variables|Default|GetHF_EjectOnImpulse))
        (Vehicle|EjectAllPlayers)))))

(fn UpdateWaterVFXOffset
  (params (Entering :type Boolean) (Lake%20Reference :type BP%20Rufus%20Water%20Object%20Reference))
  (if Entering
    (Utilities|Casting|CastToBP_Rufus_Water_LeakyLake Lake%20Reference
      (:then
        (Development|PrintString "Vehicle is entering Leaky Lake.")
        (Variables|Default|SetWaterVFXOffset -56.0))
      (:CastFailed
        (Utilities|Casting|CastToBP_Rufus_Water_LootLake Lake%20Reference
          (:then
            (Development|PrintString "Vehicle is entering Loot Lake.")
            (Variables|Default|SetWaterVFXOffset 0.0))
          (:CastFailed
            (Development|PrintString "Vehicle is entering a Rufus Water Body.")
            (Variables|Default|SetWaterVFXOffset -56.0)))))
    (else
      (Development|PrintString "Vehicle is exiting water."))))

(fn ConstructionScript
  (|CallParentFunction))
